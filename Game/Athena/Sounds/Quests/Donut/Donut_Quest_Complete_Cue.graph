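# Sound cue that, by its name, plays when a Donut quest completes in Athena, reusing the "Look At This" emote audio.
# A single wave player feeds an enveloper that holds full volume for three seconds, then fades to silence by the fourth.
# It uses the first-person emote music sound class and concurrency rules, with boosted volume and a very large max distance.
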SoundCue /Game/Athena/Sounds/Quests/Donut/Donut_Quest_Complete_Cue
root=Enveloper_0
nodes [
 {
  id=Enveloper_0
  children [WavePlayer_1]
  properties {pitchCurve {editorCurveData {keys [{time=0}]}}, volumeCurve {editorCurveData {keys [<at 0 1>,<at 3 1>,<at 4 0>]}}}
 }
 {id=WavePlayer_1, properties {soundWaveAssetPtr=/Game/Athena/Sounds/Emotes/Emote_Look_At_This_No_SPKR}}
]
properties {
 maxDistance=2097152
 soundClassObject=/Game/Sounds/FortSoundClassesAndModes/Fort_Music_Emote_1P_PSM
 volumeMultiplier=1.75
 concurrencySet [/Game/Sounds/Concurrency_Rules/Athena_EmoteMusic_1P_Conc]
}
